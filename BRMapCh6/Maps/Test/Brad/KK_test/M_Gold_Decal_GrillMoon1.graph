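# Translucent deferred decal material in a test map folder that gives a shiny gold metallic look, tinted by a Color parameter.
# A Texture parameter, defaulting to a BoatFlake splatter, has its RGB multiplied by the tint.
# Its red channel feeds a SmoothStep with adjustable min and max, forming a remappable mask.
# Scalar parameters expose high metalness and low roughness.
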Material /BRMapCh6/Maps/Test/Brad/KK_test/M_Gold_Decal_GrillMoon1
graph {
 properties {bAutomaticallySetUsageInEditor=false, blendMode=BLEND_Translucent, materialDomain=MD_DeferredDecal, usedShadingModels=Default Lit}
 nodes [
  n1 = TextureSampleParameter2D() {parameterName=Texture, texture=/Game/Global/Textures/Default/PointSplatter/BoatFlake}
  Multiply(A=VectorParameter() {parameterName=Color, defaultValue=<rgba 1 0.3681 0.0521 1>}.RGB, B=@n1.RGB)
  ScalarParameter() {defaultValue=0.92, parameterName=Metal}
  ScalarParameter() {defaultValue=0.05, parameterName=Roughness}
  SmoothStep(Min=ScalarParameter() {parameterName=SmoothStepMin}, Max=ScalarParameter() {defaultValue=1, parameterName=SmoothStepMax}, Value=@n1.R)
 ]
}
